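# Physics asset for the F_MED_Fuzzy_Bear_Donut female medium body cosmetic, set up for secondary motion on its left and right pouches.
# A pelvis capsule anchors two mirrored chains, each running through a tiny master capsule to a larger pouch capsule.
# The master joints rotate freely against the pelvis, while each pouch has tight swing and twist limits relative to its master.
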PhysicsAsset /BRCosmetics/Characters/Player/Female/Medium/Bodies/F_MED_Fuzzy_Bear_Donut/Meshes/F_MED_Fuzzy_Bear_Donut_Physics
bodies {
 dyn_pouch_l [{shapeType=Capsule, length=5.1075, radius=3.8834, center=<1.7969 -0.1758 0.1407>}]
 dyn_pouch_l_MSTR [{shapeType=Capsule, length=0.505, radius=0.505}]
 dyn_pouch_r [{shapeType=Capsule, length=5.1075, radius=3.8834, center=<1.7969 0.1758 0.1405>}]
 dyn_pouch_r_MSTR [{shapeType=Capsule, length=0.505, radius=0.505}]
 pelvis [{shapeType=Capsule, length=12.5099, radius=1.908, center=<3.1924 -0.8111 0>}]
}
constraints [
 {
  bone1Name=dyn_pouch_l_MSTR
  bone2Name=pelvis
  swing1LimitDegrees=45
  swing1Motion=Free
  swing2LimitDegrees=45
  swing2Motion=Free
  twistLimitDegrees=45
  twistMotion=Free
 }
 {bone1Name=dyn_pouch_l, bone2Name=dyn_pouch_l_MSTR, swing1LimitDegrees=15, swing2LimitDegrees=15, twistLimitDegrees=5}
 {
  bone1Name=dyn_pouch_r_MSTR
  bone2Name=pelvis
  swing1LimitDegrees=45
  swing1Motion=Free
  swing2LimitDegrees=45
  swing2Motion=Free
  twistLimitDegrees=45
  twistMotion=Free
 }
 {bone1Name=dyn_pouch_r, bone2Name=dyn_pouch_r_MSTR, swing1LimitDegrees=15, swing2LimitDegrees=15, twistLimitDegrees=5}
]
