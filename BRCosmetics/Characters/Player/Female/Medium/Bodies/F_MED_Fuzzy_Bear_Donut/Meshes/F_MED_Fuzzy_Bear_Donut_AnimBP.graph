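# Animation Blueprint for the Fuzzy Bear Donut female body, adding secondary motion on top of the pose copied from the attached parent mesh.
# Anim Dynamics drives the armband bone, additive bone transforms shift both pouches, then a RigidBody node simulates the body's physics asset.
# The event graph maps each thigh's yaw delta from reference pose, relative to pelvis, to that side's pouch alpha, offsetting pouches as legs lift.
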AnimBlueprint /BRCosmetics/Characters/Player/Female/Medium/Bodies/F_MED_Fuzzy_Bear_Donut/Meshes/F_MED_Fuzzy_Bear_Donut_AnimBP
target_skeleton=&/F_MED_Fuzzy_Bear_Donut_Skeleton
graphs [
 AnimGraph {nodes [
   n1 = Anim Dynamics - Bone: dyn_armband(ComponentPose=Local To Component(LocalPose=Copy Pose From Mesh() {bUseAttachedParent=true}), Alpha=0.75, ExternalForce="0,0,0") {
    alpha=0.9
    angularSpringConstant=1500
    bAngularSpring=true
    bGravityOverrideInSimSpace=true
    numSolverIterationsPostUpdate=4
    numSolverIterationsPreUpdate=16
    simulationSpace=Actor
    alphaBoolBlend {blendOption=Linear}
    boundBone {boneName=dyn_armband}
    physicsBodyDefinitions [{
      sphereCollisionRadius=0
      boundBone {boneName=dyn_armband}
      boxExtents=<15 5 5>
      constraintSetup {angularLimitsMax=<5 15 15>, angularLimitsMin=<-5 -15 -15>, angularTarget=<1 0 0>}
      localJointOffset=<0 0 0>
     }]
   }
   n2 = "Transform (Modify) Bone - Bone: dyn_pouch_l_MSTR"(ComponentPose=@n1, Alpha=Get thigh_up_l() -> (thigh_up_l:double), Translation="-3,0,-4", Rotation="0,0,0", Scale="1,1,1") {
    rotationMode=BMM_Additive
    translationMode=BMM_Additive
    alphaBoolBlend {blendOption=Linear}
    boneToModify {boneName=dyn_pouch_l_MSTR}
    translation {x=-3, z=-4}
   }
   n3 = "Transform (Modify) Bone - Bone: dyn_pouch_r_MSTR"(ComponentPose=@n2, Alpha=Get thigh_up_r() -> (thigh_up_r:double), Translation="-2.500,-4.000,0.000", Rotation="0,30,0", Scale="1,1,1") {
    rotationMode=BMM_Additive
    translationMode=BMM_Additive
    alphaBoolBlend {blendOption=Linear}
    boneToModify {boneName=dyn_pouch_r_MSTR}
    rotation {yaw=30}
    translation=<-2.5 -4>
   }
   n4 = RigidBody(ComponentPose=@n3) {
    bOverrideWorldGravity=true
    overridePhysicsAsset=&/F_MED_Fuzzy_Bear_Donut_Physics
    simulationSpace=BaseBoneSpace
    alphaBoolBlend {blendOption=Linear}
    baseBoneRef {boneName=root}
    componentLinearAccScale=<0.3 0.2 0.2>
    componentLinearVelScale=<0.3 0.2 0.2>
    overrideWorldGravity {z=-980}
   }
   Output Pose(Result=Component To Local(ComponentPose=@n4)) {layerGroup=None}
  ]}
 EventGraph {nodes [
   Event BlueprintUpdateAnimation()
   n1 = BreakTransform(InTransform=GetDeltaTransformFromRefPose(self=GetOwnerSkeletalMeshComponent(), BoneName=thigh_l, BaseName=pelvis))
   > Set thigh_up_l(thigh_up_l:double=MapRangeClamped(Value=BreakRotator(InRot=@n1.Rotation).Yaw, InRangeA=-20, InRangeB=-100)) -> (Output_Get:double)
   n2 = BreakTransform(InTransform=GetDeltaTransformFromRefPose(self=GetOwnerSkeletalMeshComponent(), BoneName=thigh_r, BaseName=pelvis))
   > Set thigh_up_r(thigh_up_r:double=MapRangeClamped(Value=BreakRotator(InRot=@n2.Rotation).Yaw, InRangeA=-20, InRangeB=-100)) -> (Output_Get:double)
  ]}
]
variables [
 thigh_up_l {type=number}
 thigh_up_r {type=number}
]
