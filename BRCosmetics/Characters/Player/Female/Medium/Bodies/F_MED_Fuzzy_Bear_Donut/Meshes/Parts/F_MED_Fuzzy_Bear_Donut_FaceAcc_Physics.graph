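# Physics asset for the face accessory mesh of the Fuzzy Bear Donut female medium player cosmetic, giving its two ears physical bodies.
# Capsule bodies cover the head and the left and right dynamic ear bones.
# Each ear is constrained to the head with matching swing and twist limits, with twist tighter than swing.
# No constraint links the two ears directly.
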PhysicsAsset /BRCosmetics/Characters/Player/Female/Medium/Bodies/F_MED_Fuzzy_Bear_Donut/Meshes/Parts/F_MED_Fuzzy_Bear_Donut_FaceAcc_Physics
bodies {
 dyn_ear_l [{shapeType=Capsule, length=5.337, radius=3.7662, center=<2.5 0 0>}]
 dyn_ear_r [{shapeType=Capsule, length=5.214, radius=3.8314, center=<2.5 0 0>}]
 head [{shapeType=Capsule, length=5.125, radius=10.2237, center=<5.3416 -0.2503 0.2758>}]
}
constraints [
 {bone1Name=dyn_ear_l, bone2Name=head, swing1LimitDegrees=45, swing2LimitDegrees=45, twistLimitDegrees=15}
 {bone1Name=dyn_ear_r, bone2Name=head, swing1LimitDegrees=45, swing2LimitDegrees=45, twistLimitDegrees=15}
]
